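# Sound cue for equipping the MurkyVole arm cannons, played as weapon foley with reload attenuation settings.
# A mixer layers a modulated random equip sound with a louder, slightly delayed sweetener randomly picked from two waves.
# A quality-level switch feeds the modulator, using a six-wave random pool for the first two levels and a three-wave subset for the last two.
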SoundCue /MurkyVoleGameplay/Assets/Audio/GauntletGuns/MurkyVole_ArmCannons_Equip_Cue
root=Mixer_0
nodes [
 {id=Delay_0, children [WavePlayer_2], properties {delayMax=0.1, delayMin=0.1}}
 {id=Delay_1, children [WavePlayer_3], properties {delayMax=0.1, delayMin=0.1}}
 {id=Mixer_0, children [Modulator_0,Random_2], properties {inputVolume [1,3]}}
 {id=Modulator_0, children [QualityLevel_0]}
 {id=QualityLevel_0, children [Random_0,Random_0,Random_1,Random_1]}
 {id=Random_0, children [WavePlayer_0,WavePlayer_1,WavePlayer_4,WavePlayer_5,WavePlayer_6,WavePlayer_7], properties {weights [1,1,1,1,1,1]}}
 {id=Random_1, children [WavePlayer_5,WavePlayer_6,WavePlayer_7], properties {bShouldExcludeFromBranchCulling=true, weights [1,1,1]}}
 {id=Random_2, children [Delay_0,Delay_1], properties {weights [1,1]}}
 {id=WavePlayer_0, properties {soundWaveAssetPtr=&/MurkyVole_ArmCannons_Equip_01}}
 {id=WavePlayer_1, properties {soundWaveAssetPtr=&/MurkyVole_ArmCannons_Equip_02}}
 {id=WavePlayer_2, properties {soundWaveAssetPtr=&/MurkyVole_ArmCannons_Equip_Sweetner_01}}
 {id=WavePlayer_3, properties {soundWaveAssetPtr=&/MurkyVole_ArmCannons_Equip_Sweetner_02}}
 {id=WavePlayer_4, properties {soundWaveAssetPtr=&/MurkyVole_ArmCannons_Equip_03}}
 {id=WavePlayer_5, properties {soundWaveAssetPtr=&/MurkyVole_ArmCannons_Equip_04}}
 {id=WavePlayer_6, properties {soundWaveAssetPtr=&/MurkyVole_ArmCannons_Equip_05}}
 {id=WavePlayer_7, properties {soundWaveAssetPtr=&/MurkyVole_ArmCannons_Equip_06}}
]
properties {
 attenuationSettings=&Fort_Weapon_Reload_Att
 bDebug=true
 bOverrideConcurrency=true
 maxDistance=2250
 soundClassObject=&Fort_WeaponFoley
 volumeMultiplier=0.652
}
