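# Sound cue for the overheat sound of MurkyVole's arm cannons, mixing an overheat sound with a delayed robotic glitch layer.
# A quality-level node picks a random overheat wave from either a six-variant pool or a three-variant subset, with default modulation.
# The boosted glitch layer randomly plays one of four individually volume-trimmed waves after a short fixed delay.
# The cue references shared weapon-reload attenuation and overrides concurrency to prevent new instances at the limit.
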SoundCue /MurkyVoleGameplay/Assets/Audio/GauntletGuns/MurkyVole_ArmCannons_OnOverheat_Cue
root=Mixer_0
nodes [
 {id=Delay_0, children [Random_2], properties {delayMax=0.15, delayMin=0.15}}
 {id=Mixer_0, children [Modulator_0,Modulator_2], properties {inputVolume [1,2]}}
 {id=Modulator_0, children [QualityLevel_0]}
 {id=Modulator_2, children [Delay_0], properties {pitchMax=1, pitchMin=1, volumeMax=2.5, volumeMin=2.5}}
 {id=Modulator_3, children [WavePlayer_7], properties {pitchMax=1, pitchMin=1, volumeMax=0.7, volumeMin=0.7}}
 {id=Modulator_4, children [WavePlayer_8], properties {pitchMax=1, pitchMin=1, volumeMax=0.9, volumeMin=0.9}}
 {id=Modulator_5, children [WavePlayer_9], properties {pitchMax=1, pitchMin=1, volumeMax=1.1, volumeMin=1.1}}
 {id=Modulator_6, children [WavePlayer_10], properties {pitchMax=1, pitchMin=1, volumeMax=1.1, volumeMin=1.1}}
 {id=QualityLevel_0, children [Random_0,Random_0,Random_1,Random_1]}
 {id=Random_0, children [WavePlayer_6,WavePlayer_0,WavePlayer_1,WavePlayer_3,WavePlayer_4,WavePlayer_5], properties {weights [1,1,1,1,1,1]}}
 {id=Random_1, children [WavePlayer_3,WavePlayer_4,WavePlayer_5], properties {bShouldExcludeFromBranchCulling=true, weights [1,1,1]}}
 {id=Random_2, children [Modulator_3,Modulator_4,Modulator_5,Modulator_6], properties {weights [1,1,1,1]}}
 {id=WavePlayer_0, properties {soundWaveAssetPtr=&/MurkyVole_ArmCannons_Overheat_02}}
 {id=WavePlayer_1, properties {soundWaveAssetPtr=&/MurkyVole_ArmCannons_Overheat_03}}
 {id=WavePlayer_10, properties {soundWaveAssetPtr=&/MurkyVole_ArmCannons_Overheat_RoboGlitch_04}}
 {id=WavePlayer_3, properties {soundWaveAssetPtr=&/MurkyVole_ArmCannons_Overheat_04}}
 {id=WavePlayer_4, properties {soundWaveAssetPtr=&/MurkyVole_ArmCannons_Overheat_05}}
 {id=WavePlayer_5, properties {soundWaveAssetPtr=&/MurkyVole_ArmCannons_Overheat_06}}
 {id=WavePlayer_6, properties {soundWaveAssetPtr=&/MurkyVole_ArmCannons_Overheat_01}}
 {id=WavePlayer_7, properties {soundWaveAssetPtr=&/MurkyVole_ArmCannons_Overheat_RoboGlitch_01}}
 {id=WavePlayer_8, properties {soundWaveAssetPtr=&/MurkyVole_ArmCannons_Overheat_RoboGlitch_02}}
 {id=WavePlayer_9, properties {soundWaveAssetPtr=&/MurkyVole_ArmCannons_Overheat_RoboGlitch_03}}
]
properties {
 attenuationSettings=&Fort_Weapon_Reload_Att
 bOverrideConcurrency=true
 maxDistance=2250
 soundClassObject=&Fort_WeaponRanged_Fire_3P_PSM
 volumeMultiplier=0.8
 attenuationOverrides {distanceAlgorithm=NaturalSound, falloffDistance=4000, attenuationShapeExtents {x=1000}}
 concurrencyOverrides {resolutionRule=PreventNew}
}
