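# Looping sound cue for the overheat layer of the MurkyVole arm cannons.
# A looping wave player feeds a continuous modulator driven by the Overheat parameter, which maps to a rising pitch range and sets volume directly, with a silent default.
# The cue uses close weapon-projectile attenuation, the low-ammo-critical ranged weapon sound class and a concurrency override that prevents new instances at the limit.
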SoundCue /MurkyVoleGameplay/Assets/Audio/GauntletGuns/MurkyVole_ArmCannons_OverheatLoop_Cue
root=ModulatorContinuous_1
nodes [
 {
  id=ModulatorContinuous_1
  children [WavePlayer_0]
  properties {
   pitchModulationParams {default=0, maxOutput=1.5, minOutput=0.85, parameterName=Overheat}
   volumeModulationParams {default=0, maxInput=0, paramMode=MPM_Direct, parameterName=Overheat}
  }
 }
 {
  id=WavePlayer_0
  properties {bLooping=true, soundWaveAssetPtr=&/MurkyVole_ArmCannons_Overheat_Loop}
 }
]
properties {
 attenuationSettings=&Fort_Weapon_Projectile_Close_Athena_Att
 bOverrideConcurrency=true
 maxDistance=4950
 pitchMultiplier=1.5
 soundClassObject=/Game/Sounds/FortSoundClassesAndModes/Fort_WeaponRanged_LowAmmo_Critical_PSM
 volumeMultiplier=1
 attenuationOverrides {distanceAlgorithm=NaturalSound, falloffDistance=4000, attenuationShapeExtents {x=1000}}
 concurrencyOverrides {resolutionRule=PreventNew}
}
