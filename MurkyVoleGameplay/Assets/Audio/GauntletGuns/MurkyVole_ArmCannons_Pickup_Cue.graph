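# Sound cue for picking up the MurkyVole arm cannons, routed to the loot-collect sound class with weapon-pickup attenuation and a pickup concurrency set.
# A mixer layers three parts: a modulated random pick of four pickup waves, a louder random pick of three briefly delayed second-part waves, and a later servo sound.
# The servo wave is borrowed from the HoverJets equip audio, and fixed delays stagger the layers into a short sequence.
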SoundCue /MurkyVoleGameplay/Assets/Audio/GauntletGuns/MurkyVole_ArmCannons_Pickup_Cue
root=Mixer_1
nodes [
 {id=Delay_0, children [WavePlayer_4], properties {delayMax=0.2, delayMin=0.2}}
 {id=Delay_2, children [WavePlayer_5], properties {delayMax=0.2, delayMin=0.2}}
 {id=Delay_4, children [WavePlayer_7], properties {delayMax=0.2, delayMin=0.2}}
 {id=Delay_5, children [WavePlayer_8], properties {delayMax=0.4, delayMin=0.4}}
 {id=Mixer_1, children [Modulator_0,Random_2,Delay_5], properties {inputVolume [1,2,1.5]}}
 {id=Modulator_0, children [Random_0]}
 {id=Random_0, children [WavePlayer_0,WavePlayer_1,WavePlayer_2,WavePlayer_3], properties {weights [1,1,1,1]}}
 {id=Random_2, children [Delay_0,Delay_2,Delay_4], properties {weights [1,1,1]}}
 {id=WavePlayer_0, properties {soundWaveAssetPtr=&/MurkyVole_ArmCannons_PickUp_01}}
 {id=WavePlayer_1, properties {soundWaveAssetPtr=&/MurkyVole_ArmCannons_PickUp_02}}
 {id=WavePlayer_2, properties {soundWaveAssetPtr=&/MurkyVole_ArmCannons_PickUp_03}}
 {id=WavePlayer_3, properties {soundWaveAssetPtr=&/MurkyVole_ArmCannons_PickUp_04}}
 {id=WavePlayer_4, properties {soundWaveAssetPtr=&/MurkyVole_ArmCannons_PickUp_P2_04}}
 {id=WavePlayer_5, properties {soundWaveAssetPtr=&/MurkyVole_ArmCannons_PickUp_P2_03}}
 {id=WavePlayer_7, properties {soundWaveAssetPtr=&/MurkyVole_ArmCannons_PickUp_P2_01}}
 {id=WavePlayer_8, properties {soundWaveAssetPtr=/MurkyVoleGameplay/Assets/Audio/JetPack/MurkyVole_HoverJets_Equip_Servo}}
]
properties {
 attenuationSettings=&Fort_Weapon_Pickup_Att
 maxDistance=3500
 soundClassObject=&Fort_Loot_Collect
 volumeMultiplier=0.5
 concurrencyOverrides {maxCount=1, resolutionRule=PreventNew}
 concurrencySet [&PickupMaximum_Oldest]
}
