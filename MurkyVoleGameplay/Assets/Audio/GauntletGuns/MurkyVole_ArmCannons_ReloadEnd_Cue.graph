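# Sound cue for the end of the Murky Vole arm cannons' reload, classed as weapon foley.
# A random node picks among three equally weighted reload-end waves and is excluded from branch culling.
# A modulator at the root adds a very slight upward pitch variation while keeping volume constant.
# The cue uses the shared Fort weapon reload attenuation, overrides concurrency, and has its debug flag on.
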SoundCue /MurkyVoleGameplay/Assets/Audio/GauntletGuns/MurkyVole_ArmCannons_ReloadEnd_Cue
root=Modulator_0
nodes [
 {id=Modulator_0, children [Random_0], properties {pitchMax=1.01, pitchMin=1, volumeMax=1, volumeMin=1}}
 {id=Random_0, children [WavePlayer_2,WavePlayer_8,WavePlayer_9], properties {bShouldExcludeFromBranchCulling=true, weights [1,1,1]}}
 {id=WavePlayer_2, properties {soundWaveAssetPtr=&/MurkyVole_ArmCannons_ReloadEnd_01}}
 {id=WavePlayer_8, properties {soundWaveAssetPtr=&/MurkyVole_ArmCannons_ReloadEnd_03}}
 {id=WavePlayer_9, properties {soundWaveAssetPtr=&/MurkyVole_ArmCannons_ReloadEnd_04}}
]
properties {
 attenuationSettings=&Fort_Weapon_Reload_Att
 bDebug=true
 bOverrideConcurrency=true
 maxDistance=2250
 soundClassObject=&Fort_WeaponFoley
 volumeMultiplier=1
}
